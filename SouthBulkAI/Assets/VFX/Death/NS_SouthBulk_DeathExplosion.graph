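# Niagara system for the death explosion VFX of the SouthBulk AI.
# It layers five emitters: smoke, sparks, streaks and two flashes, all with LW suffixes that suggest lightweight emitters.
# The system script uses SystemState and two Set Variables assignments, with no user parameters.
# No renderers, data interfaces or dynamic inputs are listed, so the emitters' look and timing are not visible here.
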NiagaraSystem /SouthBulkAI/Assets/VFX/Death/NS_SouthBulk_DeathExplosion
dependencies {
 usedDataInterfaces []
 usedDynamicInputs []
 usedModules [
  &self:SystemSpawnScript.SystemScriptSource.SystemScriptGraph.NiagaraNodeAssignment_1.SetVariables_#1
  &self:SystemSpawnScript.SystemScriptSource.SystemScriptGraph.NiagaraNodeAssignment_4.SetVariables_#2
  &SystemState
 ]
}
emitters {
 Flash_LW {rendererClasses [], renderers []}
 Flash_LW001 {rendererClasses [], renderers []}
 SmokeLW {rendererClasses [], renderers []}
 SparksLW {rendererClasses [], renderers []}
 StreaksLW {rendererClasses [], renderers []}
}
summary {
 systemName=NS_SouthBulk_DeathExplosion
 emitters [
  {emitterName=SmokeLW, rendererClasses []}
  {emitterName=SparksLW, rendererClasses []}
  {emitterName=Flash_LW, rendererClasses []}
  {emitterName=StreaksLW, rendererClasses []}
  {emitterName=Flash_LW001, rendererClasses []}
 ]
 userVariables []
}
