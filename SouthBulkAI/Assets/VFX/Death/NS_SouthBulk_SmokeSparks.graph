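# Niagara effect for the SouthBulk death VFX, combining electrical sparks with grey smoke puffs.
# LightningBolts spawns flickering, short-lived mesh bolts and glow sprites after a brief delay, and Smoke_Burst follows later with growing sprites faded by User.FXOpacity.
# Both drive their materials through animated dynamic parameters, and their skeletal mesh and socket placement modules are disabled.
# A disabled Blasts emitter holds a six-loop explosion sequence at array-indexed offsets, using spawn groups to vary layers.
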NiagaraSystem /SouthBulkAI/Assets/VFX/Death/NS_SouthBulk_SmokeSparks
dependencies {
 usedDataInterfaces [
  &NiagaraDataInterfaceCurve
  &NiagaraDataInterfaceArrayFloat3
  &NiagaraDataInterfaceColorCurve
  &NiagaraDataInterfacePlatformSet
  &NiagaraDataInterfaceSocketReader
 ]
 usedDynamicInputs [&MaskFloatBySpawnGroup]
 usedModules [
  &self:SystemSpawnScript.SystemScriptSource.SystemScriptGraph.NiagaraNodeAssignment_0.SetVariables_#1
  &self:SystemSpawnScript.SystemScriptSource.SystemScriptGraph.NiagaraNodeAssignment_1.SetVariables_#2
  &self:SystemSpawnScript.SystemScriptSource.SystemScriptGraph.NiagaraNodeAssignment_2.SetVariables_#3
  &self:SystemSpawnScript.SystemScriptSource.SystemScriptGraph.NiagaraNodeAssignment_3.SetVariables_#4
  &SystemState
 ]
}
emitters {
 Blasts {
  bEnabled=false
  emitterUpdateScript [
   &EmitterState(
    Life Cycle Mode=::NewEnumerator1
    Loop Behavior=::NewEnumerator2
    Loop Count=6
    Loop Duration=&FloatFromCurve(FloatCurve=DataInterface() {propertyValues {Curve {keys [<at 0 0.2>,<at 0.4 0.1>,<at 0.7 1>,<at 0.8 0.1>]}}}, CurveIndex=$Engine.System.Age)
    Recalculate Duration Each Loop=true
    UseLoopDelay=false
   )
   /SetVariables(Emitter.ExplosionPositionArray=DataInterface() {propertyValues {GpuSyncMode=SyncCpuToGpu, MaxElements=0, FloatData [<30 30 -20>, <-20 -70 -70>, <40 70 -30>, <40 -70 50>, <-50 -50 -20>, <20 -10 50>]}}) {isSetParameters=true}
   /SetVariables(Emitter.PositionRandom:NiagaraPosition=&AddVectorToPosition(
     Position=<0 0 80>
     Vector=&SelectVectorFromArray(
      Direct Array Index=$Emitter.LoopCount
      Vector Selection Array=DataInterface() {propertyValues {FloatData [<-60 0 0>, <70 30 0>, <70 -40 30>, <-50 20 50>, <-50 -30 -20>, <-10 30 100>]}}
     )
    )) {enabled=false, isSetParameters=true}
   &SpawnBurst_Instantaneous(Spawn Count=6, Spawn Group=3, Spawn Time=$Emitter.ExplosionDelay, Use Spawn Probability=false, UseLoopCountLimit=false) {module=SpawnBurst_Instantaneous003}
   &SpawnBurst_Instantaneous(Use Spawn Probability=false, UseLoopCountLimit=false)
   &SpawnBurst_Instantaneous(Spawn Count=3, Spawn Group=1, Use Spawn Probability=false, UseLoopCountLimit=false) {module=SpawnBurst_Instantaneous001}
   &SpawnBurst_Instantaneous(Spawn Count=8, Spawn Group=2, Use Spawn Probability=false, UseLoopCountLimit=false) {module=SpawnBurst_Instantaneous002}
   /SetVariables(Emitter.PositionRandom:NiagaraPosition=&AddVectorToPosition(
     Position=<0 0 200>
     Vector=&Multiply_VectorByFloat(
      Vector=&RandomRangeVector(Minimum=<-1 -1 -1>, Maximum=<1 1 1>)
      Float=&FloatFromCurve(
       FloatCurve=DataInterface() {propertyValues {Curve {keys [
           {interpMode=RCIM_Cubic, tangentMode=RCTM_User, time=0, value=1}
           {arriveTangent=-0.7881, interpMode=RCIM_Cubic, leaveTangent=-0.7881, tangentMode=RCTM_User, time=1.3, value=0.4}
          ]}}}
       CurveIndex=$System.Age
       Scale Curve=230
      )
     )
    )) {enabled=false, isSetParameters=true}
  ]
  particleSpawnScript [
   &self:GetEmitterCount() {module=GetEmitterCount}
   /SetVariables(Particles.ExplosionScale:NiagaraFloat=&Multiply_Float(
     A=$System.ExplosionScale
     B=&MergeHelper_FloatSwitch(0=1, 1=1, 2=1.5, 3=1.5, 4=2, 5=3, Index=$Particles.EmitterLoopID, Num=::NewEnumerator4)
    )) {isSetParameters=true}
   /SetVariables(Particles.ExplosionPosition:NiagaraPosition=&AddVectorToPosition(
     Position=$Engine.Emitter.SimulationPosition
     Vector=&AddVectorToPosition(
      Position=<0 0 30>
      Vector=&SelectVectorFromArray(Direct Array Index=$Particles.EmitterLoopID, Vector Selection Array=$Emitter.ExplosionPositionArray)
     )
    )) {isSetParameters=true}
   &SkeletalMeshLocation(
    Bone Sampling Mode=::NewEnumerator4
    MaskSpawnGroup=false
    Sampled Position Offset:NiagaraPosition=$Particles.ExplosionPosition
    Skeletal Mesh=$System.Skeletal Mesh Reader
   )
   &InitializeParticle(
    Color=<rgba 3 0.91 0.15 1>
    Lifetime=&MergeHelper_FloatSwitch(
     0=0.26
     1=&RandomRangeFloat(Minimum=0.85, Maximum=1.3)
     2=&RandomRangeFloat(Minimum=0.6, Maximum=0.7)
     3=?
     Index=&SpawnGroup()
     Num=::NewEnumerator1
    )
    Mesh Scale Mode=::NewEnumerator3
    Mesh Uniform Scale=3.3
    Position Mode=::NewEnumerator0
    Sprite Rotation Mode=::NewEnumerator2
    Write Material Random=true
   )
   &ShapeLocation(
    MaskSpawnGroup=false
    Override Seed=false
    Sphere Radius=&Multiply_Float(
     A=$Particles.ExplosionScale
     B=&MergeHelper_FloatSwitch(0=0, 1=15, 2=50, 3=5, Index=&SpawnGroup(), Num=::NewEnumerator2)
    )
    UseOwnerScale=true
   )
   /SetVariables(Particles.VisibilityTag=&SpawnGroup()) {isSetParameters=true}
   &AddVelocity(
    Module.Use Falloff Distance=false
    Override Seed=false
    Velocity Mode=::NewEnumerator1
    Velocity Origin=$Particles.ExplosionPosition
    Velocity Speed=&Multiply_Float(
     A=$Particles.ExplosionScale
     B=&MergeHelper_FloatSwitch(
      0=0
      1=50
      2=&RandomRangeFloat(Minimum=600, Maximum=1000)
      3=3
      Index=&SpawnGroup()
      Num=::NewEnumerator2
     )
    )
   )
   &InitialMeshOrientation(
    Mesh Axis To Orient=<0 0 1>
    Mesh Orientation Mode=::NewEnumerator2
    Rotation=&RandomRangeVector(Minimum=<0 0 0>, Maximum=<0 0 1>)
    Use Rotation Vector=true
   )
   &CameraOffset(Camera Offset Amount=100)
  ]
  particleUpdateScript [
   &ParticleState()
   /SetVariables(Particles.SpriteSize=&Multiply_Vector2D_ByFloat(
     Float=&Multiply_Float(
      A=$Particles.ExplosionScale
      B=&MergeHelper_FloatSwitch(
       0=&FloatFromCurve(FloatCurve=DataInterface() {propertyValues {Curve {keys [
            {arriveTangent=1.2438, interpMode=RCIM_Cubic, leaveTangent=1.2438, tangentMode=RCTM_User, time=0, value=0.5}
            {interpMode=RCIM_Cubic, leaveTangent=0.4972, time=0.46, value=1}
           ]}}})
       1=&FloatFromCurve(FloatCurve=DataInterface() {propertyValues {Curve {keys [
            {arriveTangent=1.4169, interpMode=RCIM_Cubic, leaveTangent=1.4169, tangentMode=RCTM_User, time=0, value=0.5}
            {interpMode=RCIM_Cubic, tangentMode=RCTM_User, time=0.5, value=1}
           ]}}})
       2=&FloatFromCurve(FloatCurve=DataInterface() {propertyValues {Curve {keys [
            {arriveTangent=-0.1507, interpMode=RCIM_Cubic, leaveTangent=-0.1507, tangentMode=RCTM_User, time=0.8, value=1}
            {arriveTangent=-7.6519, interpMode=RCIM_Cubic, leaveTangent=-7.6519, tangentMode=RCTM_User, time=1, value=0}
           ]}}})
       3=?
       Index=&SpawnGroup()
       Num=::NewEnumerator1
      )
     )
     Vector2D=&MergeHelper_Vector2Switch(
      0=&Vector2DFromFloat(Value=450)
      1=&Vector2DFromFloat(Value=&RandomRangeFloat(Minimum=200, Maximum=350))
      2=&Multiply_Vector2D_ByFloat(Float=&RandomRangeFloat(Minimum=3, Maximum=6), Vector2D=<1 3>)
      3=?
      4=?
      Index=&SpawnGroup()
      Num=::NewEnumerator1
     )
    )) {isSetParameters=true}
   &ScaleMeshSize(Scale Factor=&VectorFromFloat(Value=&FloatFromCurve(
      FloatCurve=DataInterface() {propertyValues {Curve {keys [
          {arriveTangent=-0.0684, interpMode=RCIM_Cubic, leaveTangent=-0.0684, tangentMode=RCTM_User, time=0, value=0.35}
          {arriveTangent=0.65, interpMode=RCIM_Cubic, leaveTangent=0.65, time=0.3, value=0.5}
          {arriveTangent=0.1561, interpMode=RCIM_Cubic, leaveTangent=0.1561, tangentMode=RCTM_User, time=1, value=1}
         ]}}}
      Scale Curve=$Particles.ExplosionScale
     )))
   &ScaleSpriteSize(
    Scale Sprite Size Mode=::NewEnumerator0
    Uniform Scale Factor=&Multiply_Float(
     A=$System.ExplosionScale
     B=&MergeHelper_FloatSwitch(
      0=&FloatFromCurve(FloatCurve=DataInterface() {propertyValues {Curve {keys [
           {arriveTangent=1.2438, interpMode=RCIM_Cubic, leaveTangent=1.2438, tangentMode=RCTM_User, time=0, value=0.5}
           {interpMode=RCIM_Cubic, leaveTangent=0.4972, time=0.46, value=1}
          ]}}})
      1=&FloatFromCurve(FloatCurve=DataInterface() {propertyValues {Curve {keys [
           {arriveTangent=0.4607, interpMode=RCIM_Cubic, leaveTangent=0.4607, tangentMode=RCTM_User, time=0, value=0.4}
           {interpMode=RCIM_Cubic, tangentMode=RCTM_User, time=0.5, value=1}
          ]}}})
      2=&FloatFromCurve(FloatCurve=DataInterface() {propertyValues {Curve {keys [
           {arriveTangent=-0.1507, interpMode=RCIM_Cubic, leaveTangent=-0.1507, tangentMode=RCTM_User, time=0.8, value=1}
           {arriveTangent=-7.6519, interpMode=RCIM_Cubic, leaveTangent=-7.6519, tangentMode=RCTM_User, time=1, value=0}
          ]}}})
      3=?
      Index=&SpawnGroup()
      Num=::NewEnumerator1
     )
    )
   ) {enabled=false}
   &ScaleColor(
    Linear Color Curve=DataInterface() {propertyValues {
      AlphaCurve {keys [<at 0.0428 1>,<at 0.1889 0.3>,<at 0.6478 0>]}
      BlueCurve {keys [<at 0.0239 4>,<at 0.13 1>,<at 0.4686 0.05>]}
      GreenCurve {keys [<at 0.0239 4>,<at 0.13 1>,<at 0.4686 0.145>]}
      RedCurve {keys [<at 0.0239 4>,<at 0.13 1>,<at 0.4686 1>]}
     }}
    Scale Mode=2
   )
   /SetVariables(
    Particles.ColorSpikes:LinearColor=&MultiplyLinearColor(Linear Color=$Particles.Color, Scale Factor=<0.7 0.6 0.4 0.7>)
    Particles.SpriteSizeSpikes:Vector2f=&Multiply_Vector2D_ByFloat(Float=1, Vector2D=$Particles.SpriteSize)
   ) {isSetParameters=true}
   /SetVariables(
    Particles.ColorGlow:LinearColor=&ColorFromCurve(ColorCurve=DataInterface() {propertyValues {
       AlphaCurve {keys [<at 0 0>,<at 0.1646 0.5 RCIM_Cubic>,<at 1 0 RCIM_Cubic>]}
       BlueCurve {keys [<at 0.0987 0.04 RCIM_Cubic>,<at 0.6041 0>]}
       GreenCurve {keys [<at 0.0987 0.3993 RCIM_Cubic>,<at 0.6041 0.1>]}
       RedCurve {keys [<at 0.0987 2 RCIM_Cubic>,<at 0.6041 2>]}
      }})
    Particles.SpriteSizeGlow:Vector2f=&Vector2DFromFloat(Value=&FloatFromCurve(
      FloatCurve=DataInterface() {propertyValues {Curve {keys [
          {arriveTangent=1.5712, interpMode=RCIM_Cubic, leaveTangent=1.5712, tangentMode=RCTM_User, time=0, value=0.3}
          {interpMode=RCIM_Cubic, tangentMode=RCTM_User, time=0.6, value=1}
         ]}}}
      Scale Curve=&Multiply_Float(A=500, B=$Particles.ExplosionScale)
     ))
   ) {isSetParameters=true}
   /SetVariables(Particles.ColorSmoke:LinearColor=&ColorFromCurve(ColorCurve=DataInterface() {propertyValues {
       AlphaCurve {keys [<at 0 0>,<at 0.1818 1 RCIM_Cubic>,<at 0.3108 0.5>,<at 1 0 RCIM_Cubic>]}
       BlueCurve {keys [<at 0.2076 0 RCIM_Cubic>,<at 0.9516 0>]}
       GreenCurve {keys [<at 0.2076 0 RCIM_Cubic>,<at 0.9516 0>]}
       RedCurve {keys [<at 0.2076 0 RCIM_Cubic>,<at 0.9516 0>]}
      }})) {enabled=false, isSetParameters=true}
   &Drag(
    Drag=&MergeHelper_FloatSwitch(0=0, 1=2, 2=1, 3=?, Index=&SpawnGroup(), Num=::NewEnumerator1)
    Use Linear Drag=true
    Use Rotational Drag=true
   )
   &FN_Sprite_Attributes(
    Color=&ColorFromCurve(ColorCurve=DataInterface() {propertyValues {
       AlphaCurve {keys [<at 0 0>,<at 0.0938 1 RCIM_Cubic>,<at 0.6901 0.9>,<at 0.9771 0 RCIM_Cubic>]}
       BlueCurve {keys [
         <at 0.0629 0.022>
         {arriveTangent=0.1238, interpMode=RCIM_Cubic, leaveTangent=0.1238, time=0.2219, value=0.036}
         <at 0.3699 0.06>
         <at 0.7806 0.009 RCIM_Cubic>
        ]}
       GreenCurve {keys [
         <at 0.0629 0.4213>
         {arriveTangent=-1.1554, interpMode=RCIM_Cubic, leaveTangent=-1.1554, time=0.2219, value=0.1512}
         <at 0.3699 0.0667>
         <at 0.7806 0.009 RCIM_Cubic>
        ]}
       RedCurve {keys [
         <at 0.0629 2.2>
         {arriveTangent=-6.8418, interpMode=RCIM_Cubic, leaveTangent=-6.8418, time=0.2219, value=0.9}
         <at 0.3699 0.1>
         <at 0.7806 0.01 RCIM_Cubic>
        ]}
      }})
    Create Color Attribute=true
    Create Dynamic Material Parameter Attributes=true
    Index 0 Param 1:NiagaraFloat=&Add_Float(
     A=0
     B=&FloatFromCurve(FloatCurve=DataInterface() {propertyValues {Curve {keys [
          {arriveTangent=2.7533, interpMode=RCIM_Cubic, leaveTangent=1.8958, tangentMode=RCTM_User, time=0, value=1}
          {arriveTangent=0.8292, interpMode=RCIM_Cubic, leaveTangent=0.8292, tangentMode=RCTM_User, time=0.321, value=1.3453}
          {arriveTangent=0.0529, interpMode=RCIM_Cubic, leaveTangent=0.2021, tangentMode=RCTM_User, time=1, value=1.5}
         ]}}})
    )
    Index 0 Param 3:NiagaraFloat=&FloatFromCurve(FloatCurve=DataInterface() {propertyValues {Curve {keys [
         {arriveTangent=-53.9199, interpMode=RCIM_Cubic, leaveTangent=-8.9727, tangentMode=RCTM_User, time=0, value=2}
         {arriveTangent=-1.6741, interpMode=RCIM_Cubic, leaveTangent=-1.6741, tangentMode=RCTM_User, time=0.1, value=0.25}
         {arriveTangent=0.2347, interpMode=RCIM_Cubic, leaveTangent=0.2723, tangentMode=RCTM_Break, time=0.3, value=0}
        ]}}})
    Index 0 Param 4:NiagaraFloat=&FloatFromCurve(FloatCurve=DataInterface() {propertyValues {Curve {keys [<at 0.5 1>,<at 1 0>]}}})
    Index 1 Param 1:NiagaraFloat=&FloatFromCurve(
     FloatCurve=DataInterface() {propertyValues {Curve {keys [
         {interpMode=RCIM_Cubic, leaveTangent=4.7088, tangentMode=RCTM_Break, time=0, value=0}
         {arriveTangent=4.2257, interpMode=RCIM_Cubic, leaveTangent=4.2257, tangentMode=RCTM_User, time=0.25, value=1}
        ]}}}
     Scale Curve=&RandomRangeFloat(Minimum=1.2, Maximum=2)
    )
    Index 1 Param 2:NiagaraFloat=&FloatFromCurve(
     FloatCurve=DataInterface() {propertyValues {Curve {keys [<at 0 0.2>, {leaveTangent=0.9066, time=1, value=2}]}}}
     Scale Curve=&RandomRangeFloat(Minimum=0.1, Maximum=0.2)
    )
    Index 1 Param 3:NiagaraFloat=&FloatFromCurve(FloatCurve=DataInterface() {propertyValues {Curve {keys [
         {interpMode=RCIM_Cubic, leaveTangent=6.3917, tangentMode=RCTM_Break, time=0, value=0.2}
         {arriveTangent=1.7549, interpMode=RCIM_Cubic, leaveTangent=10.6767, tangentMode=RCTM_User, time=0.25, value=1}
        ]}}})
    Index 1 Param 4:NiagaraFloat=&FloatFromCurve(
     FloatCurve=DataInterface() {propertyValues {Curve {keys [{arriveTangent=-0.136, interpMode=RCIM_Cubic, leaveTangent=-0.136, tangentMode=RCTM_User, time=0, value=1}, <at 1 0.95>]}}}
     Scale Curve=&RandomRangeFloat(Minimum=1.5, Maximum=3)
    )
    Index 2 Param 1:NiagaraFloat=&FloatFromCurve(FloatCurve=DataInterface() {propertyValues {Curve {keys [{interpMode=RCIM_Cubic, leaveTangent=2.5581, tangentMode=RCTM_Break, time=0, value=0}, {tangentMode=RCTM_User, time=1, value=1}]}}})
    Index 2 Param 3=1.5
    Index 2 Param 4=0.3
    Module Output:ENiagara_AttributeSamplingApplyOutput=::NewEnumerator1
    Module.Index0Float0Write=true
    Module.Index0Float1Write=false
    Module.Index0Float2Write=true
    Module.Index0Float3Write=true
    Write Parameter Index 1:ENiagara_DynamicMaterialParametersInputMode=::NewEnumerator1
    Write Parameter Index 2:ENiagara_DynamicMaterialParametersInputMode=::NewEnumerator1
   )
   &GravityForce(Gravity=&Multiply_VectorByFloat(
     Vector=<0 0 -980>
     Float=&MergeHelper_FloatSwitch(0=0, 1=-0.4, 2=1, 3=?, Index=&SpawnGroup(), Num=::NewEnumerator1)
    ))
   &DynamicMaterialParameters(Module.Index0Float0Write=false, Module.Index0Float1Write=false, Module.Index0Float2Write=false, Module.Index0Float3Write=false)
   /SetVariables(Particles.ColorSparks:LinearColor=&ColorFromCurve(ColorCurve=DataInterface() {propertyValues {
       AlphaCurve {keys [<at 0.769 1 RCIM_Cubic>]}
       BlueCurve {keys [<at 0.0871 1.5 RCIM_Cubic>,<at 0.3797 0>,<at 1 0 RCIM_Cubic>]}
       GreenCurve {keys [<at 0.0871 5.75 RCIM_Cubic>,<at 0.3797 3.3333>,<at 1 0.15 RCIM_Cubic>]}
       RedCurve {keys [<at 0.0871 10 RCIM_Cubic>,<at 0.3797 10>,<at 1 3 RCIM_Cubic>]}
      }})) {isSetParameters=true}
   &SolveForcesAndVelocity(Clamp Velocity=false, Limit Acceleration=false)
   /SetVariables(Particles.SpriteSizeSpikesLong:Vector2f=&Multiply_Vector2D_ByFloat(
     Float=$Particles.ExplosionScale
     Vector2D=&MakeVector2D(
      X=&RandomRangeFloat(Minimum=40, Maximum=60)
      Y=&FloatFromCurve(
       FloatCurve=DataInterface() {propertyValues {Curve {keys [{arriveTangent=1.4594, interpMode=RCIM_Cubic, leaveTangent=1.4594, tangentMode=RCTM_User, time=0, value=0.2}, <at 1 1>]}}}
       Scale Curve=&RandomRangeFloat(Minimum=170, Maximum=210)
      )
     )
    )) {isSetParameters=true}
  ]
  rendererClasses [&NiagaraSpriteRendererProperties, &NiagaraMeshRendererProperties]
  renderers [
   {rendererClass=&NiagaraSpriteRendererProperties}
   {rendererClass=&NiagaraSpriteRendererProperties, rendererIndex=1}
   {rendererClass=&NiagaraSpriteRendererProperties, rendererIndex=2}
   {rendererClass=&NiagaraSpriteRendererProperties, rendererIndex=3}
   {rendererClass=&NiagaraSpriteRendererProperties, rendererIndex=4}
   {rendererClass=&NiagaraSpriteRendererProperties, rendererIndex=5}
   {rendererClass=&NiagaraMeshRendererProperties, rendererIndex=6}
   {rendererClass=&NiagaraSpriteRendererProperties, rendererIndex=7}
  ]
 }
 LightningBolts {
  emitterUpdateScript [
   &EmitterState(Life Cycle Mode=::NewEnumerator1, Loop Behavior=::NewEnumerator1, Loop Delay=0.3, Loop Duration=2, UseLoopDelay=true)
   &SpawnRate(Spawn Probability=0.8, SpawnRate=25, Use Spawn Probability=true)
  ]
  particleSpawnScript [
   &InitializeParticle(
    Color=<rgba 1 2.3333 5 1>
    Lifetime=&RandomRangeFloat(Minimum=0.2, Maximum=0.3)
    Mesh Scale=&Multiply_VectorByFloat(
     Vector=&RandomRangeVector(Minimum=<1 1 1>, Maximum=<2 2 2>)
     Float=&RandomRangeFloat(Minimum=0.25, Maximum=0.4)
    )
    Mesh Scale Mode=::NewEnumerator4
    Sprite Rotation Mode=::NewEnumerator2
    Sprite Size Mode=::NewEnumerator3
    Uniform Sprite Size=600
    UsePositionOffset=true
    Write Material Random=true
   )
   &CameraOffset(Camera Offset Amount=50)
   &InitialMeshOrientation(Use Rotation Vector=false)
   &ShapeLocation(MaskSpawnGroup=false, Override Seed=false, Sphere Radius=60, Sphere Surface Distribution=0.619, UseOwnerScale=true) {enabled=false}
  ]
  particleUpdateScript [
   &SkeletalMeshLocation(
    Bone Sampling Mode=::NewEnumerator4
    MaskSpawnGroup=false
    Sampled Position Offset=&RandomRangeVector(Minimum=<-20 -20 -20>, Maximum=<20 20 20>)
    Skeletal Mesh=$System.Skeletal Mesh Reader
   ) {enabled=false}
   &ParticleState()
   &ScaleMeshSize(Scale Factor=&VectorFromFloat(Value=&FloatFromCurve(FloatCurve=DataInterface() {propertyValues {Curve {keys [{arriveTangent=1.5501, interpMode=RCIM_Cubic, leaveTangent=1.5501, tangentMode=RCTM_User, time=0, value=0.3}, <at 0.7 1>]}}})))
   &ScaleSpriteSize(Uniform Curve Sprite Scale=DataInterface() {propertyValues {Curve {keys [{tangentMode=RCTM_User, time=0, value=0}, <at 0.2 1>, <at 0.7 1>, {tangentMode=RCTM_User, time=1, value=0}]}}}) {enabled=false}
   &ScaleColor(
    Scale Alpha=&FloatFromCurve(
     FloatCurve=DataInterface() {propertyValues {Curve {keys [
         {interpMode=RCIM_Cubic, tangentMode=RCTM_User, time=0.3, value=1}
         {arriveTangent=-8.4324, interpMode=RCIM_Cubic, leaveTangent=-8.4324, tangentMode=RCTM_User, time=0.4461, value=0.4034}
         {arriveTangent=-3.3559, interpMode=RCIM_Cubic, leaveTangent=-3.3559, tangentMode=RCTM_User, time=0.5245, value=1}
         {arriveTangent=2.4655, interpMode=RCIM_Cubic, leaveTangent=2.4655, tangentMode=RCTM_User, time=0.7097, value=0.2284}
         {interpMode=RCIM_Cubic, tangentMode=RCTM_User, time=1, value=0}
        ]}}}
     Scale Curve=0.5
    )
    Scale RGB=&Multiply_VectorByFloat(
     Vector=&MakeVectorFromLinearColorRGB(Color=&ColorFromCurve(ColorCurve=DataInterface() {propertyValues {
         AlphaCurve {keys [<at 0 1 RCIM_Cubic>]}
         BlueCurve {keys [<at 0.0733 10 RCIM_Cubic>,<at 0.2917 1>,<at 0.8501 0.3 RCIM_Cubic>]}
         GreenCurve {keys [<at 0.0733 7.5 RCIM_Cubic>,<at 0.2917 0.8667>,<at 0.8501 0.225 RCIM_Cubic>]}
         RedCurve {keys [<at 0.0733 5 RCIM_Cubic>,<at 0.2917 0.8>,<at 0.8501 0.21 RCIM_Cubic>]}
        }}))
     Float=&FloatFromCurve(
      FloatCurve=DataInterface() {propertyValues {Curve {keys [
          {interpMode=RCIM_Cubic, tangentMode=RCTM_User, time=0.3, value=1}
          {arriveTangent=-8.4324, interpMode=RCIM_Cubic, leaveTangent=-8.4324, tangentMode=RCTM_User, time=0.4475, value=0.1667}
          {arriveTangent=-3.3559, interpMode=RCIM_Cubic, leaveTangent=-3.3559, tangentMode=RCTM_User, time=0.5245, value=1}
          {arriveTangent=6.2447, interpMode=RCIM_Cubic, leaveTangent=6.2447, tangentMode=RCTM_User, time=0.7069, value=0.4177}
          {arriveTangent=-6.5152, interpMode=RCIM_Cubic, leaveTangent=-6.5152, tangentMode=RCTM_User, time=1, value=0}
         ]}}}
      Scale Curve=0.5
     )
    )
    ScaleA=false
    ScaleRGB=true
   )
   &DynamicMaterialParameters(
    Index 0 Param 4=13
    Index 1 Param 1=&FloatFromCurve(FloatCurve=DataInterface() {propertyValues {Curve {keys [<at 0 0.5>,<at 1 1>]}}})
    Index 1 Param 2=2
    Index 1 Param 3=&FloatFromCurve(FloatCurve=DataInterface() {propertyValues {Curve {keys [{interpMode=RCIM_Cubic, tangentMode=RCTM_Break, time=0.5, value=-0.5}, <at 1 0.1>]}}})
    Index 2 Param 1=&Multiply_Float(
     A=&FloatFromCurve(FloatCurve=DataInterface() {propertyValues {Curve {keys [{interpMode=RCIM_Cubic, leaveTangent=8.1085, tangentMode=RCTM_Break, time=0.8568, value=1}, {tangentMode=RCTM_User, time=1, value=0}]}}})
     B=&Divide_Float(
      A=&MakeFloatFromVector(Vector=$Particles.Scale, Channel=::NewEnumerator0)
      B=&RandomRangeFloat(Minimum=10, Maximum=15)
     )
    )
    Index 2 Param 2=&FloatFromCurve(
     FloatCurve=DataInterface() {propertyValues {Curve {keys [
         {interpMode=RCIM_Cubic, tangentMode=RCTM_Break, time=0.7, value=0}
         {arriveTangent=-0.0859, interpMode=RCIM_Cubic, tangentMode=RCTM_Break, time=1, value=-1}
        ]}}}
     Scale Curve=50
    )
    Index 2 Param 3=&Waveform(
     Global Input Position=&RandomRangeFloat(Minimum=0, Maximum=15)
     Global Phase=$Particles.Age
     "Waveform [1]"=::NewEnumerator12
     "[1] Phase Scale"=2
     "[1] Frequency"=1
     "[1] Amplitude Min/Max"=<0 1>
    )
    Index 2 Param 4=25
    Index 3 Param 1=&FloatFromCurve(
     FloatCurve=DataInterface() {propertyValues {Curve {keys [
         {interpMode=RCIM_Cubic, leaveTangent=-0.0025, tangentMode=RCTM_Break, time=0, value=0}
         {arriveTangent=1.8528, interpMode=RCIM_Cubic, tangentMode=RCTM_Break, time=1, value=1}
        ]}}}
     Scale Curve=&RandomRangeFloat(Minimum=0, Maximum=0)
    )
    Index 3 Param 2=&FloatFromCurve(
     FloatCurve=DataInterface() {propertyValues {Curve {keys [
         {interpMode=RCIM_Cubic, leaveTangent=-1.3572, tangentMode=RCTM_Break, time=0.2, value=0.5}
         {arriveTangent=-0.0316, interpMode=RCIM_Cubic, tangentMode=RCTM_Break, time=1, value=-0.2}
        ]}}}
     Scale Curve=0.1
    )
    Index 3 Param 3=&FloatFromCurve(FloatCurve=DataInterface() {propertyValues {Curve {keys [
         {interpMode=RCIM_Cubic, leaveTangent=1.1034, tangentMode=RCTM_Break, time=0, value=-1}
         {arriveTangent=2.2399, tangentMode=RCTM_User, time=0.7, value=0.125}
        ]}}})
    Index 3 Param 4=&FloatFromCurve(FloatCurve=DataInterface() {propertyValues {Curve {keys [
         {interpMode=RCIM_Cubic, leaveTangent=3.0601, tangentMode=RCTM_Break, time=0.5, value=-0.5}
         {arriveTangent=3.2053, tangentMode=RCTM_User, time=1, value=1}
        ]}}})
    Module.Index0Param0Write=true
    Module.Index0Param1Write=true
    Module.Index0Param2Write=true
    Module.Index0Param3Write=true
    Module.Index1Param0Write=true
    Module.Index1Param1Write=true
    Module.Index1Param2Write=true
    Module.Index1Param3Write=true
    Module.Index2Param0Write=true
    Module.Index2Param1Write=true
    Module.Index2Param2Write=true
    Module.Index2Param3Write=true
    Module.Index3Param0Write=true
    Module.Index3Param1Write=true
    Module.Index3Param2Write=true
    Module.Index3Param3Write=true
    Write Parameter Index 1=true
    Write Parameter Index 2=true
    Write Parameter Index 3=true
   )
   /SetVariables(Particles.ColorGlow:LinearColor=&MultiplyLinearColor(Linear Color=$Particles.Color, Scale Factor=&MakeVector4_FromLinearColor(LinearColor=<rgba 0 0.3333 1 4>))) {isSetParameters=true}
   /SetVariables(Particles.ColorGlowSprite:LinearColor=&ColorFromCurve(ColorCurve=DataInterface() {propertyValues {
       AlphaCurve {keys [<at 0.0109 0>,<at 0.3244 0.2 RCIM_Cubic>,<at 0.9312 0 RCIM_Cubic>]}
       BlueCurve {keys [<at 0 0.07 RCIM_Cubic>]}
       GreenCurve {keys [<at 0 0.07 RCIM_Cubic>]}
       RedCurve {keys [<at 0 1 RCIM_Cubic>]}
      }})) {isSetParameters=true}
  ]
  rendererClasses [&NiagaraMeshRendererProperties, &NiagaraSpriteRendererProperties]
  renderers [
   {rendererClass=&NiagaraMeshRendererProperties, rendererIndex=0}
   {rendererClass=&NiagaraMeshRendererProperties, rendererIndex=1}
   {rendererClass=&NiagaraSpriteRendererProperties, rendererIndex=2}
  ]
 }
 Smoke_Burst {
  emitterUpdateScript [
   &EmitterState(Life Cycle Mode=::NewEnumerator1, Loop Behavior=::NewEnumerator1, Loop Delay=1.8, Loop Duration=1, UseLoopDelay=true)
   &SpawnRate(SpawnRate=30, Use Spawn Probability=false)
  ]
  particleSpawnScript [
   &PartitionParticles(Number Of Partitions=3)
   &InitializeParticle(
    Color=<rgba 0.1 0.1 0.1 0.7>
    Lifetime Max=0.8
    Lifetime Min=0.5
    Lifetime Mode=::NewEnumerator1
    Sprite Size Mode=::NewEnumerator1
    Uniform Sprite Size Max=70
    Uniform Sprite Size Min=45
    UsePositionOffset=true
    Write Material Random=true
   )
   &SocketLocation(
    Interpolate Socket Sampling=true
    SamplePosOffset=false
    Socket Reader=$System.Socket Reader
    Socket Sampling Mode=::NewEnumerator3
    VelocitySpeedLimit=false
   ) {enabled=false}
   &SkeletalMeshLocation(Bone Sampling Mode=::NewEnumerator4, MaskSpawnGroup=false, Skeletal Mesh=$System.Skeletal Mesh Reader) {enabled=false}
   &CameraOffset(Camera Offset Amount=30)
  ]
  particleUpdateScript [
   &Drag(Drag=&RandomRangeFloat(Minimum=2, Maximum=4), Use Linear Drag=true, Use Rotational Drag=true) {enabled=false}
   &AccelerationForce(Acceleration=<0 0 300>) {enabled=false}
   &InheritSourceMovement(ApplyForce=false, ApplyPosition=true, ApplyPositionFalloff=false) {enabled=false}
   &SolveForcesAndVelocity(Clamp Velocity=false, Limit Acceleration=false) {enabled=false}
   &ParticleState()
   &ScaleSpriteSize(
    Uniform Curve Scale=&RandomRangeFloat(Minimum=2, Maximum=4)
    Uniform Curve Sprite Scale=DataInterface() {propertyValues {Curve {keys [{arriveTangent=1.9788, interpMode=RCIM_Cubic, leaveTangent=1.9788, tangentMode=RCTM_User, time=0, value=0.4}, <at 1 1.2 RCIM_Cubic>]}}}
   )
   &DynamicMaterialParameters(
    Index 0 Param 1=&Add_Float(
     A=&Multiply_Float(A=$Engine.Time, B=0.5)
     B=&RandomRangeFloat(Minimum=0, Maximum=1)
    )
    Index 0 Param 3=&FloatFromCurve(FloatCurve=DataInterface() {propertyValues {Curve {keys [<at 0 0.5>,<at 0.5 0>,<at 0.5 0>]}}})
    Index 0 Param 4=1
    Index 1 Param 1=0.8
    Index 1 Param 2=&RandomRangeFloat(Minimum=0, Maximum=1)
    Index 1 Param 3=1
    Index 1 Param 4=1.5
    Index 2 Param 1=1
    Index 2 Param 2=&RandomRangeFloat(Minimum=0, Maximum=1)
    Index 2 Param 3=&FloatFromCurve(FloatCurve=DataInterface() {propertyValues {Curve {keys [<at 0 0 RCIM_Cubic>, {arriveTangent=2.1666, tangentMode=RCTM_User, time=1, value=1}]}}})
    Index 2 Param 4=1
    Module.Param0WriteEnabled=true
    Module.Param0WriteEnabled1=true
    Module.Param0WriteEnabled2=true
    Module.Param1WriteEnabled=true
    Module.Param1WriteEnabled1=true
    Module.Param1WriteEnabled2=true
    Module.Param2WriteEnabled=true
    Module.Param2WriteEnabled1=true
    Module.Param2WriteEnabled2=true
    Module.Param3WriteEnabled=true
    Module.Param3WriteEnabled1=true
    Module.Param3WriteEnabled2=true
    Write Parameter Index 1=true
    Write Parameter Index 2=true
   )
   &ScaleColor(
    Scale Alpha=&FloatFromCurve(
     FloatCurve=DataInterface() {propertyValues {Curve {keys [
         {interpMode=RCIM_Cubic, tangentMode=RCTM_User, time=0, value=0}
         <at 0.2 0.8 RCIM_Cubic>
         {interpMode=RCIM_Cubic, tangentMode=RCTM_User, time=1, value=0}
        ]}}}
     Scale Curve=&FN_PickFloatDependOnPlatform(TargetPlatformActive=1, TargetPlatformNoActive=1.5)
    )
    Scale RGB=&MakeVectorFromLinearColorRGB(Color=&ColorFromCurve(ColorCurve=DataInterface() {propertyValues {
        AlphaCurve {keys []}
        BlueCurve {keys [<at 0.0233 0.5312 RCIM_Cubic>]}
        GreenCurve {keys [<at 0.0233 0.5312 RCIM_Cubic>]}
        RedCurve {keys [<at 0.0233 0.5312 RCIM_Cubic>]}
       }}))
    ScaleA=true
    ScaleRGB=true
   )
   &ScaleColor(Scale Alpha=$User.FXOpacity, ScaleA=true, ScaleRGB=true) {module=ScaleColor001}
  ]
  rendererClasses [&NiagaraSpriteRendererProperties]
  renderers [
   {rendererClass=&NiagaraSpriteRendererProperties}
   {rendererClass=&NiagaraSpriteRendererProperties, rendererIndex=1}
  ]
 }
}
summary {
 systemName=NS_SouthBulk_SmokeSparks
 emitters [
  {
   bEnabled=false
   emitterName=Blasts
   rendererClasses [&NiagaraSpriteRendererProperties, &NiagaraMeshRendererProperties]
  }
  {emitterName=Smoke_Burst, rendererClasses [&NiagaraSpriteRendererProperties]}
  {emitterName=LightningBolts, rendererClasses [&NiagaraMeshRendererProperties, &NiagaraSpriteRendererProperties]}
 ]
 userVariables [
  User.FXOpacity {type=NiagaraFloat, defaultValue=1, description=""}
  User.NormalImpact {type=Vector3f, description="", defaultValue {struct=Vector3f, value=<0 0 0>}}
  User.SkeletalMesh {
   type=NiagaraDataInterfaceSkeletalMesh
   defaultValue {
    struct=NiagaraExt_VariableValue_DataInterface
    value {dataInterface=&self:NiagaraDataInterfaceSkeletalMesh_0}
   }
  }
 ]
}
